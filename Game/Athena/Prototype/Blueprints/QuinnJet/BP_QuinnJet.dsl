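; Prototype QuinnJet actor that flies in formation from the battle bus start to a landing POI, then enables that dynamic foundation.
; Server packs bus position, rotation and forward vector (in scale) into one transform, whose RepNotify builds a seeded-random spline offset by jet index.
; Flight and landing timelines move the mesh along that spline, then a POI-aligned landing spline.
; Cosmetic updates drive contrails, dust-off, layered pass-by audio and proximity shake and rumble.
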(event EventBeginPlay
  ;[0 0 0] Begin play
  (Networking|SwitchHasAuthority
    (:Authority
      (Class|FortAsyncActionCurrentPlaylistReady|PlaylistDataReadyAsync
        (:OnReady
          (for _array_element (Variables|/Script/FortniteGame.FortGameStateAthena|GetTeamFlightPaths (Fortnite|Athena|GameState|GetGameStateAthena))
            (bind _output_get (Variables|SplineFlightProperties|SetBusStartPos (Utilities|Struct|BreakAircraftFlightInfo _array_element)))
            (bind _output_get_2 (Variables|SplineFlightProperties|SetFlightFwdVec (Math|Vector|GetForwardVector (.FlightStartRotation (Utilities|Struct|BreakAircraftFlightInfo _array_element)))))
            (bind _output_get_3 (Variables|SplineFlightProperties|SetBusStartRot (.FlightStartRotation (Utilities|Struct|BreakAircraftFlightInfo _array_element))))
            (Variables|Default|SetPackedBusTransform (Math|Transform|MakeTransform _output_get _output_get_3 _output_get_2))))))))

(event Custom|SetDestinationPOI
  (params (DestinationPOI :type Building%20Foundation%20Object%20Reference))
  ;[0 0 0] Target landing poi from controller
  (Variables|QuinJetProperties|SetMyLandingPOI DestinationPOI))

(event Custom|DebugOriginLocation
  (params (World%20Loc :type Vector)))

(event Custom|CalculateSplinePath
  (bind _numjetstotal (Variables|QuinJetProperties|GetNumJetsTotal))
  (bind _returnvalue (Math|Conversions|ToFloat(Integer) _numjetstotal))
  (bind _returnvalue_2 (Math|Float|SafeDivide _returnvalue 2.0))
  (bind _jetindex (Variables|QuinJetProperties|GetJetIndex))
  (bind _returnvalue_3 (Math|Conversions|ToFloat(Integer) _jetindex))
  (bind _returnvalue_4 (< _returnvalue_3 _returnvalue_2))
  ;[0 0 0] build our spline path
  (Variables|SplineFlightProperties|SetbIsLeftSide _returnvalue_4)
  (Variables|SplineFlightProperties|SetRandomFirstTurn (Math|Random|RandomBoolwithWeightfromStream (Variables|Default|GetMyRandomStream) 0.7))
  (bind _output_get (Variables|SplineFlightProperties|SetInitialPos (+ (select _returnvalue_4 (* (* (Math|Vector|CrossProduct (Variables|SplineFlightProperties|GetFlightFwdVec) "0, 0,1.000000") -1.0) (* (Math|Float|Absolute(Float) (- _returnvalue_2 (Math|Conversions|ToFloat(Integer) (Variables|QuinJetProperties|GetJetIndex)))) (Variables|SplineFlightProperties|GetJetsFormationDistance_Horizontal))) (* (Math|Vector|CrossProduct (Variables|SplineFlightProperties|GetFlightFwdVec) "0, 0,1.000000") (* (Math|Float|Absolute(Float) (- _returnvalue_2 (Math|Conversions|ToFloat(Integer) (Variables|QuinJetProperties|GetJetIndex)))) (Variables|SplineFlightProperties|GetJetsFormationDistance_Horizontal)))) (+ (* (* (Variables|SplineFlightProperties|GetFlightFwdVec) -1.0) (+ 1.5e+04 1.0)) (Math|Vector|MakeVector (.x (Variables|SplineFlightProperties|GetBusStartPos)) (.y (Variables|SplineFlightProperties|GetBusStartPos)) (- (.z (Variables|SplineFlightProperties|GetBusStartPos)) (- (Variables|Default|GetBusPosZOffsetMax) (* (Math|Conversions|ToFloat(Integer) (- 1 (Variables|QuinJetProperties|GetJetIndex))) 1e+03))))))))
  (Variables|SplineFlightProperties|SetNormalizedDistanceOverIslandToTarget (Math|Float|Clamp(Float) (Math|Float|SafeDivide (Math|Vector|VectorLength (- _output_get (Variables|Landing|GetDestinationPos))) (Variables|SplineFlightProperties|GetMaxDistanceCapForFlightSpeed))))
  (Utilities|IsValid (Variables|Default|GetMyFlightSplinePath)
    (:Is%20Valid
      (Spline|ClearSplinePoints (Variables|Default|GetMyFlightSplinePath) false))
    (:Is%20Not%20Valid
      (bind _returnvalue_5 (AddComponent|NODE_AddSplineComponent-0))
      (bind _output_get_2 (Variables|Default|SetMyFlightSplinePath _returnvalue_5))))
  (Spline|SetSplinePoints _output_get_2 (CallFunction|GenerateSplinePointsforFlight) "World" true)
  ;;[0 0 0] scale tangent here for no curvature at close point
  (Spline|SetTangentatSplinePoint (Variables|Default|GetMyFlightSplinePath) 1 (* (Spline|GetTangentatSplinePoint (Variables|Default|GetMyFlightSplinePath) 1 "World") 0.2) "World" true))

(event Custom|BeginSpoolUpStarted)

(event Custom|DebugTargetLandingLocation
  (params (WorldLoc :type Vector))
  ;[1 0.1563 0.619] debug
  (bind _output_get (Variables|Landing|SetDestinationPos WorldLoc))
  (Transformation|SetWorldLocation (CallFunction|GetLandingSpline) (+ _output_get (Variables|/Script/Engine.SceneComponent|GetRelativeLocation (CallFunction|GetLandingSpline))) false false))

(event Custom|BeginOnFlightStarted
  ;[0 0 0] cosmetic FX state events
  (Development|SetHiddeninGame (Variables|Default|GetSM_QuinnJet) false true)
  (Components|Activation|Activate (Variables|Default|GetPS_Hightower_Contrails_Tip_R) false))

(event Custom|BeginOnLandedStarted
  ;[0 0 0] cosmetic FX state events
  (Components|Activation|Deactivate (Variables|Default|GetPS_Hightower_Contrails_Tip_R))
  (Components|Activation|Activate (Variables|Default|GetPS_Hightower_GroundEffect) true)
  (Components|Activation|Activate (Variables|Default|GetPS_Hightower_Hover_R) false)
  (Audio|SpawnSoundAttached "/Game/Athena/Sounds/Vehicles/QuinnJet/Veh_Hightower_QJ_Land_Cue.Veh_Hightower_QJ_Land_Cue" (Variables|Default|GetSM_QuinnJet))
  (Audio|Components|Audio|FadeOut (Variables|Default|GetLoopingAudio) 3.0 0.0 "Linear" :!class /Script/Engine.AudioComponent))

(event Custom|BeginOnLanded
  ;[0 0 0] cosmetic FX state events
  (Components|Activation|Deactivate (Variables|Default|GetPS_Hightower_Hover_R))
  (Components|Activation|Activate (Variables|Default|GetPS_Hightower_GroundEffect_Blast) true))

(event Custom|BeginFlightPath
  ;[0 0 0] Begin the main flight sequence - Kick the tires and light the fires
  (CallFunction|FlyAlongSpline))

(event OnOneshotEvent(LoopingAudio)
  (params (AudioComponent :type Audio%20Component%20Object%20Reference) (ParameterName :type Name :default "None") (OneshotIndex :type Integer))
  ;[0 0 0] Passby whoosh event
  (if (Utilities|Name|Equal(Name) ParameterName "PassBy")
    (Variables|Passby|SetbHasTriggeredPassBy true)
    (Utilities|FlowControl|Delay (Variables|Passby|GetProjectedPassByTimeDelay))
    (Development|PrintString ">>>>>>>>>>>>>> QJ: Passby Event" true true "(R=1.000000,G=0.079997,B=0.000000,A=1.000000)")))

(event Custom|FlyAlongSpline
  (bind _ps_hightower_groundeffect (Variables|Default|GetPS_Hightower_GroundEffect))
  ;[0 0 0] Fly along the spline path to target destination
  ;;[0 0 0] time duration of main flight
  (Components|Timeline|SetPlayRate (Variables|BP_QuinnJet|GetMainFlightPath) (/ 1.0 (select (> (Math|Random|RandomFloatinRangefromStream (Variables|Default|GetMyRandomStream) (- (Math|Float|Lerp (Variables|SplineFlightProperties|GetFlightTimelineDurationMinDistance) (Variables|SplineFlightProperties|GetFlightTimelineDurationMaxDistance) (Variables|SplineFlightProperties|GetNormalizedDistanceOverIslandToTarget)) (Variables|SplineFlightProperties|GetFlightTimelineDurationRandomVariation)) (+ (Math|Float|Lerp (Variables|SplineFlightProperties|GetFlightTimelineDurationMinDistance) (Variables|SplineFlightProperties|GetFlightTimelineDurationMaxDistance) (Variables|SplineFlightProperties|GetNormalizedDistanceOverIslandToTarget)) (Variables|SplineFlightProperties|GetFlightTimelineDurationRandomVariation))) 0.0) (Math|Random|RandomFloatinRangefromStream (Variables|Default|GetMyRandomStream) (- (Math|Float|Lerp (Variables|SplineFlightProperties|GetFlightTimelineDurationMinDistance) (Variables|SplineFlightProperties|GetFlightTimelineDurationMaxDistance) (Variables|SplineFlightProperties|GetNormalizedDistanceOverIslandToTarget)) (Variables|SplineFlightProperties|GetFlightTimelineDurationRandomVariation)) (+ (Math|Float|Lerp (Variables|SplineFlightProperties|GetFlightTimelineDurationMinDistance) (Variables|SplineFlightProperties|GetFlightTimelineDurationMaxDistance) (Variables|SplineFlightProperties|GetNormalizedDistanceOverIslandToTarget)) (Variables|SplineFlightProperties|GetFlightTimelineDurationRandomVariation))) 1.0)))
  (Variables|Landing|SetbBeginLevelOut false)
  ;;[0 0 0] fly a path to near our target on the generated spline
  (bind (_direction _alpha _deceleratealpha) (Timeline|MainFlightPath Play)
    (:Update
      (CallFunction|UpdateFlightwithSplinePath _alpha (Variables|Default|GetMyFlightSplinePath) _deceleratealpha)
      (CallFunction|UpdateforBusPassBy)
      (CallFunction|UpdateforProximityFX)
      (Utilities|FlowControl|DoOnce false
        (:Completed
          (Networking|SwitchHasAuthority
            (:Remote
              (CallFunction|BeginonFlightStarted))))))
    (:Finished
      (CallFunction|BeginonLandedStarted)
      ;;[0 0 0] land at our target with landing spline
      (bind (_direction_2 _alpha_2 _decelerate) (Timeline|Landing Play)
        (:Update
          (CallFunction|UpdateFlightwithSplinePath _alpha_2 (CallFunction|GetLandingSpline) _decelerate true true)
          (Networking|SwitchHasAuthority
            (:Remote
              ;;[0 0 0] place dust off FX
              (bind (_outhit _returnvalue) (Collision|LineTraceByChannel (Transformation|GetWorldLocation (Variables|Default|GetSM_QuinnJet)) (+ (Transformation|GetWorldLocation (Variables|Default|GetSM_QuinnJet)) (* "0, 0,-1.000000" (Variables|QuinJetProperties|GetDustoffTraceDistance))) "TraceTypeQuery10"))
              (if _returnvalue
                ;;;[1 1 1] FX - Dustoff World loc
                (Rendering|Debug|DrawDebugSphere (Transformation|GetActorLocation (Game|GetPlayerController 0)) 5e+02 12 "(R=0.000000,G=0.599820,B=1.000000,A=1.000000)")
                (Transformation|SetWorldLocation _ps_hightower_groundeffect (.Location (Collision|BreakHitResult _outhit)) false false)
                (Transformation|SetWorldRotation _ps_hightower_groundeffect (Math|Rotator|MakeRotfromXZ (* (Transformation|GetForwardVector (Variables|Default|GetSM_QuinnJet)) "1.000000,1.000000,0.000000") (0.0 0.0 1.0)) false false)))))
        (:Finished
          (Transformation|SetWorldLocationAndRotation (Variables|Default|GetSM_QuinnJet) (Variables|Landing|GetDestinationPos) (Variables|Landing|GetDestinationRot) false false)
          (CallFunction|BeginonLanded)
          (Utilities|FlowControl|Delay (Variables|Landing|GetLandingDelayBeforePOIActivation))
          (Networking|SwitchHasAuthority
            (:Authority
              ;;[0 0 0] enable the poi to begin streaming in
              (Class|BuildingFoundation|SetDynamicFoundationEnabled (Variables|QuinJetProperties|GetMyLandingPOI) true)
              ;;[0 0 0] cleanup ourselves
              (if (Variables|QuinJetProperties|GetbDestroyAfterLand)
                (Rendering|SetActorHiddenInGame true)
                (Actor|SetLifeSpan 5.0)))
            (:Remote
              (if (Variables|QuinJetProperties|GetbDestroyAfterLand)
                (Utilities|FlowControl|Delay (Variables|Landing|GetTimeBeforeActorHidden))
                (Rendering|SetActorHiddenInGame true)))))))
    (:BeginLevelOut
      (Variables|Landing|SetbBeginLevelOut true))))

(fn GenerateSplinePointsForFlight
  ; Index 2
  (LocalVariables|SetIndex2 (+ (Math|Vector|MakeVector (.x (Variables|SplineFlightProperties|GetBusStartPos)) (.y (Variables|SplineFlightProperties|GetBusStartPos)) (- (.z (Variables|SplineFlightProperties|GetBusStartPos)) (Math|Float|SelectFloat (Math|Random|RandomFloatinRangefromStream (Variables|Default|GetMyRandomStream) 2e+03 5e+03) (Math|Random|RandomFloatinRangefromStream (Variables|Default|GetMyRandomStream) -3e+03 -3e+03) (or (== (Variables|QuinJetProperties|GetJetIndex) 0) (== (Math|Integer|%25%28Integer%29 (Variables|QuinJetProperties|GetJetIndex) 2) 0))))) (* (Math|Vector|SelectVector (Math|Vector|SelectVector (Math|Vector|RotateVectorAroundAxis (Variables|SplineFlightProperties|GetFlightFwdVec) 2e+01 "0, 0,1.000000") (Math|Vector|RotateVectorAroundAxis (Variables|SplineFlightProperties|GetFlightFwdVec) -2e+01 "0.000000,0.000000,1.000000") (Variables|SplineFlightProperties|GetbIsLeftSide)) (Variables|SplineFlightProperties|GetFlightFwdVec) (and (or (== (Variables|QuinJetProperties|GetJetIndex) 0) (== (Math|Integer|%25%28Integer%29 (Variables|QuinJetProperties|GetJetIndex) 2) 0)) (Math|Random|RandomBoolwithWeightfromStream (Variables|Default|GetMyRandomStream) 0.73))) (Math|Random|RandomFloatinRangefromStream (Variables|Default|GetMyRandomStream) (Math|Float|Lerp 2e+04 5e+04 (Variables|SplineFlightProperties|GetNormalizedDistanceOverIslandToTarget)) (Math|Float|Lerp 7e+04 9e+04 (Variables|SplineFlightProperties|GetNormalizedDistanceOverIslandToTarget))))))
  (LocalVariables|SetIndex1 (+ (* (Math|Vector|RotateVector (Variables|SplineFlightProperties|GetFlightFwdVec) (Math|Rotator|MakeRotator 0.0 0.0 (* (Math|Random|RandomFloatinRangefromStream (Variables|Default|GetMyRandomStream) -5.0 5.0) (Variables|SplineFlightProperties|GetNormalizedDistanceOverIslandToTarget)))) (Math|Float|Lerp 3.5e+04 7e+04 (Variables|SplineFlightProperties|GetNormalizedDistanceOverIslandToTarget))) (Variables|SplineFlightProperties|GetInitialPos)))
  (LocalVariables|SetIndex3 (Math|Vector|MakeVector (.x (+ (* (Math|Vector|Normalize (+ (Math|Vector|Normalize (select (not (Variables|SplineFlightProperties|GetbIsLeftSide)) (Math|Vector|RotateVectorAroundAxis (- (Spline|GetLocationatSplinePoint (CallFunction|GetLandingSpline) 0 "World") (LocalVariables|GetIndex2)) 45.0 "0, 0,1.000000") (- (Spline|GetLocationatSplinePoint (CallFunction|GetLandingSpline) 0 "World") (LocalVariables|GetIndex2)))) (Variables|SplineFlightProperties|GetFlightFwdVec))) (Math|Float|Lerp 3e+04 1e+05 (Variables|SplineFlightProperties|GetNormalizedDistanceOverIslandToTarget))) (LocalVariables|GetIndex2))) (.y (+ (* (Math|Vector|Normalize (+ (Math|Vector|Normalize (select (not (Variables|SplineFlightProperties|GetbIsLeftSide)) (Math|Vector|RotateVectorAroundAxis (- (Spline|GetLocationatSplinePoint (CallFunction|GetLandingSpline) 0 "World") (LocalVariables|GetIndex2)) 45.0 "0, 0,1.000000") (- (Spline|GetLocationatSplinePoint (CallFunction|GetLandingSpline) 0 "World") (LocalVariables|GetIndex2)))) (Variables|SplineFlightProperties|GetFlightFwdVec))) (Math|Float|Lerp 3e+04 1e+05 (Variables|SplineFlightProperties|GetNormalizedDistanceOverIslandToTarget))) (LocalVariables|GetIndex2))) (- (+ (.z (LocalVariables|GetIndex2)) (Math|Random|RandomFloatinRangefromStream (Variables|Default|GetMyRandomStream) 1.5e+03 9e+03)) (* (- (.z (Variables|SplineFlightProperties|GetInitialPos)) (.z (Variables|Landing|GetDestinationPos))) 0.2))))
  (LocalVariables|SetIndex4 (+ (* (Math|Vector|Normalize (- (Spline|GetLocationatSplinePoint (CallFunction|GetLandingSpline) 0 "World") (LocalVariables|GetIndex2))) (* (Math|Vector|VectorLength (- (Spline|GetLocationatSplinePoint (CallFunction|GetLandingSpline) 0 "World") (LocalVariables|GetIndex2))) (Math|Float|Lerp 0.5 0.85 (Variables|SplineFlightProperties|GetNormalizedDistanceOverIslandToTarget)))) (LocalVariables|GetIndex2)))
  (return (select (< (Variables|SplineFlightProperties|GetNormalizedDistanceOverIslandToTarget) 0.3) (Utilities|Array|MakeArray (Variables|SplineFlightProperties|GetInitialPos) (LocalVariables|GetIndex1) (LocalVariables|GetIndex2) (LocalVariables|GetIndex3) (LocalVariables|GetIndex4) (+ (LocalVariables|GetIndex4) (Math|Vector|MakeVector (.x (* (+ (Math|Vector|Normalize (- (LocalVariables|GetIndex3) (LocalVariables|GetIndex4))) (Math|Vector|Normalize (- (LocalVariables|GetIndex4) (Spline|GetLocationatSplinePoint (CallFunction|GetLandingSpline) 0 "World")))) (/ (Math|Vector|VectorLength (- (LocalVariables|GetIndex4) (Spline|GetLocationatSplinePoint (CallFunction|GetLandingSpline) 0 "World"))) 2.0))) (.y (* (+ (Math|Vector|Normalize (- (LocalVariables|GetIndex3) (LocalVariables|GetIndex4))) (Math|Vector|Normalize (- (LocalVariables|GetIndex4) (Spline|GetLocationatSplinePoint (CallFunction|GetLandingSpline) 0 "World")))) (/ (Math|Vector|VectorLength (- (LocalVariables|GetIndex4) (Spline|GetLocationatSplinePoint (CallFunction|GetLandingSpline) 0 "World"))) 2.0))) (/ (- (.z (LocalVariables|GetIndex4)) (.z (Spline|GetLocationatSplinePoint (CallFunction|GetLandingSpline) 0 "World"))) 0.4))) (Spline|GetLocationatSplinePoint (CallFunction|GetLandingSpline) 0 "World")) (Utilities|Array|MakeArray (Variables|SplineFlightProperties|GetInitialPos) (LocalVariables|GetIndex1) (LocalVariables|GetIndex2) (LocalVariables|GetIndex3) (LocalVariables|GetIndex4) (Spline|GetLocationatSplinePoint (CallFunction|GetLandingSpline) 0 "World")))))

(fn GetLandingSpline
  (return (Variables|Default|GetLandingSpline_B)))

(fn OnRep_MyLandingPOI
  (bind _mylandingpoi (Variables|QuinJetProperties|GetMyLandingPOI))
  (bind _returnvalue (Transformation|GetActorRotation _mylandingpoi))
  (bind _landingpoirotoffset (Variables|Landing|GetLandingPOIRotOffset))
  (bind _returnvalue_2 (Math|Rotator|CombineRotators _returnvalue _landingpoirotoffset))
  (Utilities|IsValid _mylandingpoi
    (:Is%20Valid
      (Transformation|SetWorldRotation (CallFunction|GetLandingSpline) _returnvalue_2 false false)
      (Variables|Landing|SetDestinationPos (+ (Math|Vector|RotateVector (Variables|Landing|GetLandingPOIPositionOffset) (Math|Rotator|MakeRotator 0.0 0.0 (.yaw _returnvalue))) (Transformation|GetActorLocation _mylandingpoi)))
      (Variables|Landing|SetDestinationRot _returnvalue_2)
      (Transformation|SetWorldLocationAndRotation (CallFunction|GetLandingSpline) (+ (+ (.location (Transformation|GetActorTransform (Variables|QuinJetProperties|GetMyLandingPOI))) (Variables|Landing|GetLandingPOIPositionOffset)) (Variables|/Script/Engine.SceneComponent|GetRelativeLocation (CallFunction|GetLandingSpline))) (Math|Rotator|CombineRotators _returnvalue_2 "0.000000,90.000000, 0") false false)
      (Spline|SetLocationatSplinePoint (CallFunction|GetLandingSpline) (- (Spline|GetNumberOfSplinePoints (CallFunction|GetLandingSpline)) 1) (Variables|Landing|GetDestinationPos) "World" true))))

(fn OnRep_PackedBusTransform
  (bind _packedbustransform (Variables|Default|GetPackedBusTransform))
  (Variables|SplineFlightProperties|SetBusStartPos (.location _packedbustransform))
  (Variables|SplineFlightProperties|SetBusStartRot (.rotation _packedbustransform))
  (Variables|SplineFlightProperties|SetFlightFwdVec (.scale _packedbustransform))
  (CallFunction|CalculateSplinePath)
  (CallFunction|PrePositionandHideMesh))

(fn PrePositionAndHideMesh
  (bind _sm_quinnjet (Variables|Default|GetSM_QuinnJet))
  (Transformation|SetWorldLocationAndRotation _sm_quinnjet (Variables|SplineFlightProperties|GetInitialPos) (Variables|SplineFlightProperties|GetBusStartRot) false false)
  (Networking|SwitchHasAuthority
    (:Remote
      (Development|SetHiddeninGame _sm_quinnjet true true))))

(fn UpdateFlightWithSplinePath
  (params (Alpha :type Float%20%28double-precision%29) (Spline :type Spline%20Component%20Object%20Reference) (NoseUpAlpha :type Float%20%28double-precision%29) (bControlledLanding :type Boolean) (bRotateToLand :type Boolean))
  (bind _returnvalue (Spline|GetNumberOfSplinePoints Spline))
  (bind _returnvalue_2 (- _returnvalue 4))
  (bind _returnvalue_3 (Spline|GetDistanceAlongSplineatSplinePoint Spline _returnvalue_2))
  (bind _returnvalue_4 (Spline|GetSplineLength Spline))
  (bind _returnvalue_5 (* _returnvalue_4 Alpha))
  (bind _sm_quinnjet (Variables|Default|GetSM_QuinnJet))
  (Utilities|IsValid Spline
    (:Is%20Valid
      (if bControlledLanding
        (if (> _returnvalue_5 _returnvalue_3)
          (if (not (Variables|Landing|GetbLanding))
            (Variables|Landing|SetbLanding true)
            (Variables|Landing|SetLandingStartTransform (Transformation|GetWorldTransform _sm_quinnjet)))
          (Transformation|SetWorldTransform _sm_quinnjet (Math|Transform|Lerp(Transform) (Variables|Landing|GetLandingStartTransform) (Math|Transform|MakeTransform (Variables|Landing|GetDestinationPos) (Math|Rotator|CombineRotators (Variables|Landing|GetDestinationRot) (Math|Rotator|MakeRotator (Math|Float|Lerp 0.0 3e+01 NoseUpAlpha)))) (/ (- _returnvalue_5 _returnvalue_3) (- (Spline|GetDistanceAlongSplineatSplinePoint Spline (- _returnvalue 1)) _returnvalue_3)) "DualQuatInterp") false false)
          (else
            _)))
      (Transformation|SetWorldLocationAndRotation (Variables|Default|GetSM_QuinnJet) (Math|Interpolation|VInterpTo (Transformation|GetWorldLocation (Variables|Default|GetSM_QuinnJet)) (Spline|GetLocationatDistanceAlongSpline Spline _returnvalue_5 "World") (Utilities|Time|GetWorldDeltaSeconds) 4.0) (Math|Interpolation|RInterpTo (Transformation|GetWorldRotation (Variables|Default|GetSM_QuinnJet)) (select (Variables|Landing|GetbBeginLevelOut) (Math|Rotator|CombineRotators (Variables|Landing|GetDestinationRot) "0,50.000000,20.000000") (Math|Rotator|CombineRotators (Math|Rotator|MakeRotator (.roll (Math|Rotator|FindLookatRotation (Transformation|GetWorldLocation (Variables|Default|GetSM_QuinnJet)) (select bRotateToLand (Math|Vector|MakeVector (.x (Variables|Landing|GetDestinationPos)) (.y (Variables|Landing|GetDestinationPos)) (.z (+ (Transformation|GetWorldLocation (Variables|Default|GetSM_QuinnJet)) (* (Transformation|GetForwardVector (Variables|Default|GetSM_QuinnJet)) -5e+01)))) (Spline|GetLocationatDistanceAlongSpline Spline (* _returnvalue_4 (Math|Float|Clamp(Float) (+ Alpha (Variables|SplineFlightProperties|GetLookAheadDistance)))) "World")))) (Math|Float|Clamp(Float) (Math|Float|Lerp 6e+01 -6e+01 (Math|Float|MapRangeClamped (* (Math|Vector|DotProduct (Math|Vector|CrossProduct "0, 0,1.000000" (Transformation|GetRightVector (Variables|Default|GetSM_QuinnJet))) (Math|Vector|Normalize (- (Transformation|GetWorldLocation (Variables|Default|GetSM_QuinnJet)) (Spline|GetLocationatDistanceAlongSpline Spline (* (Spline|GetSplineLength Spline) (Math|Float|Clamp(Float) (+ Alpha 0.05))) "World")))) (Variables|SplineFlightProperties|GetExagerateYawScalar)) -1.0 1.0 0.0 1.0)) -6e+01 6e+01) (.yaw (Math|Rotator|FindLookatRotation (Transformation|GetWorldLocation (Variables|Default|GetSM_QuinnJet)) (select bRotateToLand (Math|Vector|MakeVector (.x (Variables|Landing|GetDestinationPos)) (.y (Variables|Landing|GetDestinationPos)) (.z (+ (Transformation|GetWorldLocation (Variables|Default|GetSM_QuinnJet)) (* (Transformation|GetForwardVector (Variables|Default|GetSM_QuinnJet)) -5e+01)))) (Spline|GetLocationatDistanceAlongSpline Spline (* _returnvalue_4 (Math|Float|Clamp(Float) (+ Alpha (Variables|SplineFlightProperties|GetLookAheadDistance)))) "World"))))) "0,90.000000, 0")) (Utilities|Time|GetWorldDeltaSeconds) (Math|Float|SelectFloat 6.0 3.0 (Variables|Landing|GetbBeginLevelOut))) false false))))

(fn UpdateForBusPassBy
  (Networking|CanExecuteCosmeticEvents
    (:True
      (if (not (Variables|Passby|GetbHasTriggeredPassBy))
        (Utilities|FlowControl|Sequence
          (:then_0
            (if (not (Audio|Components|Audio|IsPlaying (Variables|Default|GetLoopingAudio)))
              (Audio|Components|Audio|FadeIn (Variables|Default|GetLoopingAudio) 0.5 1.0 0.0 "Linear" :!class /Script/Engine.AudioComponent)
              (for _array_element (Variables|Default|GetSoundsToPrime)
                (Audio|PrimeSound _array_element))))
          (:then_1
            (bind _output_get (LocalVariables|SetProjectedActorLocation (- (Transformation|GetWorldLocation (Variables|Default|GetSM_QuinnJet)) (* (Transformation|GetRightVector (Variables|Default|GetSM_QuinnJet)) (Variables|Passby|GetPassByTriggerAheadDistance)))))
            (bind _listenerposition (Audio|GetClosestListenerLocation _output_get 1e+05 true))
            (Variables|Default|SetListnerPos _listenerposition)
            (LocalVariables|SetPassByDot (Math|Float|Sign(Float) (Math|Vector|DotProduct (Math|Vector|Normalize (- (Variables|Default|GetListnerPos) (LocalVariables|GetProjectedActorLocation))) (* (Transformation|GetRightVector (Variables|Default|GetSM_QuinnJet)) -1.0))))
            (Class|FortLayeredAudioComponent|SetFloatParameterExt (Variables|Default|GetLoopingAudio) "PassBy" (Math|Float|MapRangeClamped (LocalVariables|GetPassByDot) -1.0 1.0 0.0 1.0))))))))

(fn UpdateForProximityFX
  (bind _sm_quinnjet (Variables|Default|GetSM_QuinnJet))
  (bind _returnvalue (Transformation|GetWorldLocation _sm_quinnjet))
  (bind _listnerpos (Variables|Default|GetListnerPos))
  (bind _returnvalue_2 (- _returnvalue _listnerpos))
  (bind _returnvalue_3 (Math|Vector|VectorLength _returnvalue_2))
  (bind _returnvalue_4 (Game|GetPlayerController 0))
  (Networking|SwitchHasAuthority
    (:Remote
      (if (< _returnvalue_3 (Variables|QuinJetProperties|GetProximityFXTriggerDistance))
        (Utilities|IsValid _returnvalue_4
          (:Is%20Valid
            (Game|Feedback|ClientPlayForceFeedback _returnvalue_4 (Variables|QuinJetProperties|GetForceFeedbackEffect) "None" false false false))
          (:Is%20Not%20Valid
            _))
        (CameraShake|Start (Variables|Default|GetCameraShakeSource))
        (elif (> _returnvalue_3 2e+04)
          (Variables|Default|SetbCanFireProximityFX true))))))

(fn ConstructionScript)
