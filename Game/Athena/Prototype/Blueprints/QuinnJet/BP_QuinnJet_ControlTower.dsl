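; Quinn Jet control tower that spawns a server-side jet fleet and launches it on flight paths in the Aircraft game phase.
; Once playlist data is ready, it randomly picks 5x5 POIs, enables their foundations and sends jets to mapped 3x3 POIs, or uses an older single-POI list.
; Replicated selected POIs make clients reveal index-matched smoke signal FX.
; Debug options override jet count and landing targets, or enable all landing POIs.
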(event Custom|TestQuinnJets
  (for _array_element (Variables|Default|GetMyQuinnJetFleet)
    (Class|BPQuinnJet|BeginFlightPath _array_element)))

(event Custom|OnGamePhaseChanged
  (params (GamePhase :type EAthenaGamePhase%20Enum))
  (Utilities|Enabled
    (:True
      (if (not (Variables|Default|GetbDeployedFleet))
        (if (Utilities|Enum|Equal(Enum) GamePhase "Aircraft")
          (for _array_element (Variables|Default|GetMyQuinnJetFleet)
            (Class|BPQuinnJet|BeginFlightPath _array_element)))))))

(event EventEndPlay
  (params (EndPlayReason :type EEndPlayReason%20Enum :default "Destroyed"))
  ;[0 0 0] begin play - spawn jets
  (bind _gamephaseupdatedasynctask (Variables|Default|GetGamePhaseUpdatedAsyncTask)
    (:then
      (GameplayEventRouter|StopListeningtoStatefulEvent _gamephaseupdatedasynctask))
    (:else)))

(event Custom|OnJetDestroyed
  (params (DestroyedActor :type Actor%20Object%20Reference))
  (bind _myquinnjetfleet (Variables|Default|GetMyQuinnJetFleet))
  ;[0 0 0] On jet destroyed
  (Networking|SwitchHasAuthority
    (:Authority
      (bind _asbp_quinn_jet (Utilities|Casting|CastToBP_QuinnJet DestroyedActor))
      (if (Utilities|Array|ContainsItem _myquinnjetfleet _asbp_quinn_jet)
        (bind _returnvalue (Utilities|Array|RemoveItem _myquinnjetfleet _asbp_quinn_jet))))))

(event EventBeginPlay
  ;[0 0 0] begin play - spawn jets
  (Networking|SwitchHasAuthority
    (:Authority
      (bind _returnvalue (GameplayEventRouter|GetEventRouter "/Script/FortniteGame.FortGameState" self))
      (bind (_asynctaskproxy _payload) (GameplayEventRouter|StartListeningToStatefulEvent _returnvalue "/Script/FortniteGame.FortGamePhaseUpdatedEvent")
        (:then
          (Variables|Default|SetGamePhaseUpdatedAsyncTask _asynctaskproxy)
          ;;[0 0 0] move to playlist ready to avoid random value determinism from server sharding
          (Class|FortAsyncActionCurrentPlaylistReady|PlaylistDataReadyAsync
            (:OnReady
              (Utilities|Enabled
                (:True
                  (CallFunction|CreateJetFleet))))))
        (:OnEventReceived
          _)
        (:OnSavedEventStateExists
          _))
      (CallFunction|OnGamePhaseChanged (Utilities|Struct|BreakFortGamePhaseUpdatedEvent _payload)))))

(event Custom|CreateJetFleet
  (bind _keys_5x5poi (Variables|Default|GetKeys_5x5POI))
  (bind _returnvalue (Utilities|Array|Length _keys_5x5poi))
  (bind _designatedlandingpois (Variables|QuinnJetControlTower|GetDesignatedLandingPOIs))
  (bind _returnvalue_2 (Utilities|Array|Length _designatedlandingpois))
  (bind _randomstream (Variables|Default|GetRandomStream))
  (bind _returnvalue_3 (Math|Random|RandomIntegerfromStream _randomstream _returnvalue))
  (bind _returnvalue_4 (- _returnvalue_2 1))
  (bind _returnvalue_5 (Math|Random|RandomIntegerinRange 0 _returnvalue_4))
  (bind _bdebugjets (Variables|QuinnJetControlTower|GetbDebugJets))
  (bind _rownumrandompoisenabled (Variables|QuinnJetSpawning|GetRowNumRandomPOIsEnabled))
  (bind _returnvalue_6 (ScalableFloat|GetValueatLevel _rownumrandompoisenabled))
  (bind _returnvalue_7 (Math|Float|Truncate _returnvalue_6))
  (bind _numdebugjetspawns (Variables|QuinnJetControlTower|GetNumDebugJetSpawns))
  (bind _returnvalue_8 (Math|Integer|SelectInt _numdebugjetspawns _returnvalue_7 _bdebugjets))
  (bind _output (Utilities|Array|Get(acopy) _designatedlandingpois _returnvalue_5))
  (bind _output_2 (Utilities|Array|Get(acopy) _keys_5x5poi _returnvalue_3))
  (bind _landinglocationpoimap (Variables|QuinnJetControlTower|GetLandingLocationPOIMap))
  (bind _chosenpoi (Variables|Default|GetChosenPOI))
  (bind _chosenpoi_2 (Variables|Default|GetChosenPOI))
  ;[0 0 0] create jet fleet - one actor jet for each random selection
  (Utilities|Enabled
    (:True
      (Networking|SwitchHasAuthority
        (:Authority
          (Math|Random|SetRandomStreamSeed (Variables|Default|GetRandomStream) (Math|DateTime|GetMillisecond (Math|DateTime|UTCNow)))
          (bind _keys (Utilities|Map|Keys (Variables|QuinnJetControlTower|GetLandingLocationPOIMap)))
          (Variables|Default|SetKeys_5x5POI _keys)
          (for _array_element (Variables|Default|GetMyQuinnJetFleet)
            (Actor|DestroyActor _array_element))
          (Utilities|Array|Clear (Variables|Default|GetMyQuinnJetFleet))
          (Utilities|Array|Clear (Variables|Default|GetAccum5POIs))
          (if (> _returnvalue_8 0)
            (for _index (range _returnvalue_8)
              (bind _returnvalue_9 (Game|SpawnActorfromClass "/Game/Athena/Prototype/Blueprints/QuinnJet/BP_QuinnJet.BP_QuinnJet_C" (Math|Transform|MakeTransform (+ (* (Variables|QuinnJetControlTower|GetJetParkedOffset) _index) (Transformation|GetWorldLocation (Variables|Default|GetTestOrigin)))) "Undefined" "OverrideRootScale" 0 _index (Variables|QuinnJetControlTower|GetbDestroyJetsAfterLanding) _returnvalue_8 0))
              (Utilities|Array|Add (Variables|Default|GetMyQuinnJetFleet) _returnvalue_9)
              (Game|BindEventtoOnDestroyed _returnvalue_9 (EventDispatchers|CreateEvent :!event Custom|OnJetDestroyed)))
            (if (Variables|Default|GetbUsePairings)
              (Fortnite|FortShippingLog self (Utilities|String|Append "QuinJetController: " (Utilities|String|ToString(Integer) (Utilities|Array|Length (Variables|Default|GetKeys_5x5POI))) "keys 5x5"))
              ;[0 0 0] Uses map for poi pairings (5x5 - 3x3)
              (for _array_element_2 (Variables|Default|GetMyQuinnJetFleet)
                (if (> _returnvalue 0)
                  (if (Utilities|Array|IsValidIndex _keys_5x5poi _returnvalue_3)
                    (Utilities|IsValid _output_2
                      (:Is%20Valid
                        (bind _output_get (Variables|Default|SetChosenPOI _output_2))
                        (Class|BuildingFoundation|SetDynamicFoundationEnabled _output_get true)
                        (if (.ReturnValue (Utilities|Map|Find _landinglocationpoimap _output_get))
                          (Class|BPQuinnJet|SetDestinationPOI _array_element_2 (Utilities|Map|Find _landinglocationpoimap _output_get))
                          (if (not (Variables|QuinnJetControlTower|GetbDebugJets))
                            (Utilities|Array|Add (Variables|Default|GetAccum5POIs) _chosenpoi)
                            (Utilities|Array|RemoveItem (Variables|Default|GetKeys_5x5POI) _chosenpoi)
                            (Variables|Default|SetChosenPOI 0))
                          (else
                            (Development|PrintString "QuinJet - 3x3 not found" true true "(R=1.000000,G=0.000000,B=0.411846,A=1.000000)" 6.0))))))))
              ;[0 0 0] Set selected pois with notify to turn on client side smoke FX
              (Variables|Default|SetSelectedPOIs (Variables|Default|GetAccum5POIs))
              (else
                ;[0 0 0] old behavior, single 5x5 poi
                (for (_array_element_3 _array_index) (Variables|Default|GetMyQuinnJetFleet)
                  (if (Utilities|Array|IsValidIndex _designatedlandingpois _returnvalue_5)
                    (if (not (Variables|QuinnJetControlTower|GetbDebugJets))
                      (bind _output_get_2 (Variables|Default|SetChosenPOI _output))
                      (Utilities|IsValid _output_get_2
                        (:Is%20Valid
                          (Class|BuildingFoundation|SetDynamicFoundationEnabled _chosenpoi_2 true)
                          (Class|BPQuinnJet|SetDestinationPOI _array_element_3 _chosenpoi_2)))
                      (else
                        (Class|BPQuinnJet|DebugTargetLandingLocation _array_element_3 (Transformation|GetWorldLocation (Variables|Default|GetTestTarget)))))
                    (Development|PrintString (Utilities|String|Append "QuinJet #" (Utilities|String|ToString(Integer) _array_index) " landing @ " (Utilities|GetDisplayName _output)) true true "(R=0.000000,G=0.660000,B=1.000000,A=1.000000)" 2e+01)
                    (if (not (Variables|QuinnJetControlTower|GetbDebugJets))
                      (Utilities|Array|RemoveItem (Variables|QuinnJetControlTower|GetDesignatedLandingPOIs) (Variables|Default|GetChosenPOI))
                      (Variables|Default|SetChosenPOI 0))))))))))))

(event Custom|EnableAllJetLandings
  (bind _landinglocationpoimap (Variables|QuinnJetControlTower|GetLandingLocationPOIMap))
  ;[0 0 0] debug cheat - turn on all jet landing pois
  (Networking|SwitchHasAuthority
    (:Authority
      (for _array_element (Variables|Default|GetKeys_5x5POI)
        (Utilities|IsValid _array_element
          (:Is%20Valid
            (Class|BuildingFoundation|SetDynamicFoundationEnabled _array_element true)
            (if (.ReturnValue (Utilities|Map|Find _landinglocationpoimap _array_element))
              (Class|BuildingFoundation|SetDynamicFoundationEnabled (Utilities|Map|Find _landinglocationpoimap _array_element) true))))))))

(fn EnableAssociatedSignalSmokeFX
  (params (BuildingFoundation :type Building%20Foundation%205x%205%20Object%20Reference))
  (LocalVariables|SetbFoundFX false)
  (bind _keys (Utilities|Map|Keys (Variables|QuinnJetControlTower|GetLandingLocationPOIMap)))
  (for (_array_element _array_index) _keys
    (if (Utilities|Equal(Object) _array_element BuildingFoundation)
      (for (_array_element_2 _array_index_2) (Variables|QuinnJetControlTower|GetSmokeSignalFX)
        (Utilities|IsValid (Utilities|ResolveSoftReference _array_element_2)
          (:Is%20Valid
            (if (== _array_index _array_index_2)
              (LocalVariables|SetbFoundFX true)
              (Fortnite|FortShippingLog self (Utilities|String|Append "QuinJetController: " "Enabled FX  " (Utilities|GetDisplayName (Utilities|ResolveSoftReference _array_element_2)) " : " (Utilities|GetDisplayName _array_element)))
              (Rendering|SetActorHiddenInGame (Utilities|ResolveSoftReference _array_element_2) false)
              (Class|BPHightowerSmokesignal|EnableNearParticles (Utilities|ResolveSoftReference _array_element_2))))))))
  (if (not (LocalVariables|GetbFoundFX))
    (Fortnite|FortShippingLog self (Utilities|String|Append "QuinJetController: " "Could not find FX for: " (Utilities|GetDisplayName BuildingFoundation) ""))))

(fn OnRep_SelectedPOIs
  (bind _smokesignalfx (Variables|QuinnJetControlTower|GetSmokeSignalFX))
  (bind _returnvalue (Utilities|Array|Length _smokesignalfx))
  (Networking|SwitchHasAuthority
    (:Authority
      (for _array_element (Variables|Default|GetSelectedPOIs)
        (Fortnite|FortShippingLog self (Utilities|String|Append "QuinJetController: " (Utilities|GetDisplayName _array_element) " enabled"))))
    (:Remote
      (if (> _returnvalue 1)
        (Fortnite|FortShippingLog self (Utilities|String|Append "QuinJetController: " (Utilities|String|ToString(Integer) _returnvalue) "fx found"))
        (else
          (Fortnite|FortShippingLog self "QuinJetController: no associated smoke fx found")))
      (for _array_element_2 (Variables|Default|GetSelectedPOIs)
        (CallFunction|EnableAssociatedSignalSmokeFX _array_element_2)))))

(fn ConstructionScript
  (if (Variables|QuinnJetControlTower|GetbDebugJets)
    _))
